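# Control Rig function library of spline helpers: SplineFromItems, AttachChainToSpline, SplineIK, SplineThrough3Points and SplineThrough4Points.
# AttachChainToSpline and SplineIK fit chains to point-built splines, optionally stretched.
# SplineIK needs four or more controls, aims bone up axes at an offset secondary spline for twist, then constrains the last bone to the last control.
# The Through functions solve cubic Bezier equations for four control points making curves pass through given points.
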ControlRigBlueprint /ControlRigSpline/SplineFunctionLibrary/SplineFunctionLibrary
graphs [
 ContainedGraph {nodes [
   n1 = RigVMFunctionEntryNode()
   > GetTransforms = RigVMCollapseNode(Array=@n1.Items)
   > RigVMFunctionReturnNode(
    Spline=/Script/ControlRigSpline.RigUnit_ControlRigSplineFromPoints(
     Compression=0
     Stretch=0
     bClosed=False
     SamplesPerSegment=@n1.Samples Per Segment
     SplineMode=@n1.Spline Mode
     Points=@GetTransforms.positions
    ) {is=struct RigVMUnitNode}.Spline
    Transforms=@GetTransforms.transforms
   )
  ]}
 ContainedGraph {nodes [
   n1 = RigVMFunctionEntryNode()
   > RigVMDispatchNode(Array=RigVMVariableNode() {variable=transformArray}.Value) {template="DISPATCH_RigVMDispatch_ArrayReset(io Array)"}
   > RigVMDispatchNode(Array=RigVMVariableNode() {variable=posArray}.Value) {template="DISPATCH_RigVMDispatch_ArrayReset(io Array)"}
   > RigVMDispatchNode(Array=RigVMVariableNode() {variable=scaleArray}.Value) {template="DISPATCH_RigVMDispatch_ArrayReset(io Array)"}
   > RigVMDispatchNode(Array=RigVMVariableNode() {variable=rotArray}.Value) {template="DISPATCH_RigVMDispatch_ArrayReset(io Array)"}
   > n2 = RigVMDispatchNode(Array=@n1.Array) {template="DISPATCH_RigVMDispatch_ArrayIterator(in Array,out Element,out Index,out Count,out Ratio)"}
   n3 = &RigUnit_GetTransform(Item=@n2.Element)
   > RigVMDispatchNode(Element=@n3.Transform, Array=RigVMVariableNode() {variable=transformArray}.Value) {template="DISPATCH_RigVMDispatch_ArrayAdd(io Array,in Element,out Index)"}
   > RigVMDispatchNode(Element=@n3.Transform.Rotation, Array=RigVMVariableNode() {variable=rotArray}.Value) {template="DISPATCH_RigVMDispatch_ArrayAdd(io Array,in Element,out Index)"}
   > RigVMDispatchNode(Element=@n3.Transform.Translation, Array=RigVMVariableNode() {variable=posArray}.Value) {template="DISPATCH_RigVMDispatch_ArrayAdd(io Array,in Element,out Index)"}
   > RigVMDispatchNode(Element=@n3.Transform.Scale3D, Array=RigVMVariableNode() {variable=scaleArray}.Value) {template="DISPATCH_RigVMDispatch_ArrayAdd(io Array,in Element,out Index)"}
   RigVMFunctionReturnNode(
    ExecuteContext=@n2.Completed
    transforms=RigVMVariableNode() {variable=transformArray}.Value
    positions=RigVMVariableNode() {variable=posArray}.Value
    scales=RigVMVariableNode() {variable=scaleArray}.Value
    rotations=RigVMVariableNode() {variable=rotArray}.Value
   )
  ]}
 ContainedGraph {nodes [
   BranchNode_RigVMUnitNode = &RigVMFunction_ControlFlowBranch(ExecuteContext=(), Condition=RigVMVariableNode() {variable=debug}.Value)
   /Script/ControlRigSpline.RigUnit_DrawControlRigSpline(
    Color=(R=1,G=0,B=0,A=1)
    Detail=16
    Thickness=0.75
    ExecutePin=@BranchNode_RigVMUnitNode.True
    Spline=RigVMVariableNode() {variable=template_spline}.Value
   ) {is=struct RigVMUnitNode}
   n1 = RigVMFunctionEntryNode()
   n2 = RigVMVariableNode() {variable=Controls}
   BranchNode_4_RigVMUnitNode = &RigVMFunction_ControlFlowBranch(
    ExecuteContext=@n1.ExecuteContext
    Condition=/Script/RigVM.RigVMFunction_MathIntLess(B=4, A=RigVMDispatchNode(Array=@n2.Value) {template="DISPATCH_RigVMDispatch_ArrayGetNum(in Array,out Num)"}.Num) {is=struct RigVMUnitNode, template="Less::Execute(in A,in B,out Result)"}.Result
   )
   RigVMFunctionReturnNode(ExecuteContext=@BranchNode_4_RigVMUnitNode.Completed, Output Spline=RigVMVariableNode() {variable=local_spline}.Value)
   RigUnit_SequenceExecution_1 = RigVMAggregateNode(ExecuteContext=@BranchNode_4_RigVMUnitNode.False)
   n3 = /Script/ControlRig.RigUnit_GetTransformItemArray(Space=GlobalSpace, bInitial=False, Items=RigVMVariableNode() {variable=Controls}.Value) {is=struct RigVMUnitNode}
   RigVMFunction_MathTransformArrayToSRT = /Script/RigVM.RigVMFunction_MathTransformArrayToSRT(Transforms=@n3.Transforms) {is=struct RigVMUnitNode}
   RigVMVariableNode(
    Value=&RigUnit_ControlRigSplineFromPoints(
     bClosed=False
     Points=@RigVMFunction_MathTransformArrayToSRT.Translations
     SplineMode=RigVMVariableNode() {variable=Spline Mode}.Value
     SamplesPerSegment=RigVMVariableNode() {variable=Samples Per Segment}.Value
    ).Spline
    ExecuteContext=@RigUnit_SequenceExecution_1.A
   ) {variable=template_spline}
   > RigVMVariableNode(Value=@n3.Transforms) {variable=template_transforms}
   > RigUnit_SequenceExecution = RigVMAggregateNode()
   n4 = RigVMVariableNode() {variable=template_transforms}
   RigVMDispatchNode(
    Num=RigVMDispatchNode(Array=@n4.Value) {template="DISPATCH_RigVMDispatch_ArrayGetNum(in Array,out Num)"}.Num
    ExecuteContext=@RigUnit_SequenceExecution.A
    Array=RigVMVariableNode() {variable=offset_positions}.Value
   ) {template="DISPATCH_RigVMDispatch_ArraySetNum(io Array,in Num)"}
   > n5 = RigVMDispatchNode(Array=@n4.Value) {template="DISPATCH_RigVMDispatch_ArrayIterator(in Array,out Element,out Index,out Count,out Ratio)"}
   > RigVMDispatchNode(
    Index=@n5.Index
    Element=/Script/RigVM.RigVMFunction_MathTransformTransformVector(Transform=@n5.Element, Location=RigVMVariableNode() {variable=Secondary Spline Direction}.Value) {is=struct RigVMUnitNode}.Result
    Array=RigVMVariableNode() {variable=offset_positions}.Value
   ) {template="DISPATCH_RigVMDispatch_ArraySetAtIndex(io Array,in Index,in Element)"}
   RigUnit_SequenceExecution_4 = RigVMAggregateNode(ExecuteContext=@RigUnit_SequenceExecution.C)
   Branch_RigVMUnitNode = &RigVMFunction_ControlFlowBranch(ExecuteContext=@RigUnit_SequenceExecution_4.A, Condition=RigVMVariableNode() {variable=Stretch}.Value)
   n6 = RigVMVariableNode() {variable=template_spline}
   n7 = RigVMVariableNode() {variable=Primary Axis}
   n8 = RigVMVariableNode() {variable=Bones}
   /Script/ControlRigSpline.RigUnit_FitChainToSplineCurveItemArray(
    Alignment=Front
    DebugSettings=(Scale=1,CurveColor=(R=1,G=1,B=0,A=1),SegmentsColor=(R=1,G=0,B=0,A=1),WorldOffset=(Rotation=(X=0,Y=0,Z=0,W=1),Translation=(X=0,Y=0,Z=0),Scale3D=(X=1,Y=1,Z=1)))
    Maximum=1
    PoleVectorPosition=(X=0,Y=0,Z=0)
    Rotations=()
    SamplingPrecision=12
    SecondaryAxis=(X=0,Y=0,Z=0)
    Weight=1
    bPropagateToChildren=True
    ExecutePin=@Branch_RigVMUnitNode.False
    Spline=@n6.Value
    PrimaryAxis=@n7.Value
    Items=@n8.Value
   ) {is=struct RigVMUnitNode}
   &RigUnit_FitChainToSplineCurveItemArray(
    Alignment=Stretched
    DebugSettings=(Scale=1,CurveColor=(R=1,G=1,B=0,A=1),SegmentsColor=(R=1,G=0,B=0,A=1),WorldOffset=(Rotation=(X=0,Y=0,Z=0,W=1),Translation=(X=0,Y=0,Z=0),Scale3D=(X=1,Y=1,Z=1)))
    Maximum=1
    PoleVectorPosition=(X=0,Y=0,Z=0)
    Rotations=()
    SamplingPrecision=12
    SecondaryAxis=(X=0,Y=0,Z=0)
    Weight=1
    bPropagateToChildren=True
    ExecutePin=@Branch_RigVMUnitNode.True
    Spline=@n6.Value
    PrimaryAxis=@n7.Value
    Items=@n8.Value
   )
   RigUnit_SequenceExecution_2 = RigVMAggregateNode(ExecuteContext=@RigUnit_SequenceExecution_4.B)
   n9 = RigVMVariableNode(
    Value=&RigUnit_ControlRigSplineFromPoints(
     Stretch=0
     bClosed=False
     SplineMode=RigVMVariableNode() {variable=Spline Mode}.Value
     SamplesPerSegment=RigVMVariableNode() {variable=Samples Per Segment}.Value
     Points=RigVMVariableNode() {variable=offset_positions}.Value
    ).Spline
    ExecuteContext=@RigUnit_SequenceExecution_2.A
   ) {variable=local_spline}
   BranchNode_2_RigVMUnitNode = &RigVMFunction_ControlFlowBranch(ExecuteContext=@n9.ExecuteContext, Condition=RigVMVariableNode() {variable=debug}.Value)
   &RigUnit_DrawControlRigSpline(
    Color=(R=1,G=0,B=0,A=1)
    Detail=16
    Thickness=1
    ExecutePin=@BranchNode_2_RigVMUnitNode.True
    Spline=RigVMVariableNode() {variable=local_spline}.Value
   )
   n10 = RigVMVariableNode() {variable=Bones}
   DISPATCH_RigVMDispatch_ArrayGetNum = RigVMDispatchNode(Array=@n10.Value) {template="DISPATCH_RigVMDispatch_ArrayGetNum(in Array,out Num)"}
   DISPATCH_RigVMDispatch_ArraySetNum = RigVMDispatchNode(
    Num=@DISPATCH_RigVMDispatch_ArrayGetNum.Num
    ExecuteContext=@RigUnit_SequenceExecution_2.B
    Array=RigVMVariableNode() {variable=aim_transforms}.Value
   ) {template="DISPATCH_RigVMDispatch_ArraySetNum(io Array,in Num)"}
   > n11 = RigVMDispatchNode(Array=@n10.Value) {template="DISPATCH_RigVMDispatch_ArrayIterator(in Array,out Element,out Index,out Count,out Ratio)"}
   n12 = &RigUnit_GetTransform(Item=@n11.Element)
   n13 = RigVMVariableNode() {variable=local_spline}
   > DISPATCH_RigVMDispatch_ArraySetAtIndex = RigVMDispatchNode(
    Index=@n11.Index
    Element=/Script/ControlRigSpline.RigUnit_PositionFromControlRigSpline(
     U=/Script/ControlRigSpline.RigUnit_ClosestParameterFromControlRigSpline(Position=@n12.Transform.Translation, Spline=@n13.Value) {is=struct RigVMUnitNode}.U
     Spline=@n13.Value
    ) {is=struct RigVMUnitNode}.Position
    Array=RigVMVariableNode() {variable=aim_transforms}.Value
   ) {template="DISPATCH_RigVMDispatch_ArraySetAtIndex(io Array,in Index,in Element)"}
   n14 = RigVMDispatchNode(ExecuteContext=@RigUnit_SequenceExecution.D, Array=RigVMVariableNode() {variable=Bones}.Value) {template="DISPATCH_RigVMDispatch_ArrayIterator(in Array,out Element,out Index,out Count,out Ratio)"}
   DISPATCH_RigVMDispatch_ArrayGetAtIndex = RigVMDispatchNode(Index=@n14.Index, Array=RigVMVariableNode() {variable=aim_transforms}.Value) {template="DISPATCH_RigVMDispatch_ArrayGetAtIndex(in Array,in Index,out Element)"}
   n15 = /Script/ControlRig.RigUnit_AimBoneMath(
    DebugSettings=(Scale=10,WorldOffset=(Rotation=(X=0,Y=0,Z=0,W=1),Translation=(X=0,Y=0,Z=0),Scale3D=(X=1,Y=1,Z=1)))
    Primary=(Weight=1,Axis=(X=0,Y=1,Z=0),Target=(X=1,Y=0,Z=0),Kind=Location,Space=(Type=Bone))
    Secondary=(Weight=0,Axis=(X=0,Y=0,Z=0),Target=(X=0,Y=0,Z=0),Kind=Location,Space=(Type=Bone))
    Weight=1
    Primary.Target=@DISPATCH_RigVMDispatch_ArrayGetAtIndex.Element
    InputTransform=&RigUnit_GetTransform(Item=@n14.Element).Transform
    Primary.Axis=RigVMVariableNode() {variable=Up Axis}.Value
   ) {is=struct RigVMUnitNode}
   MathIntEquals = RigVMDispatchNode(
    A=@n14.Index
    B=/Script/RigVM.RigVMFunction_MathIntSub(B=1, A=@n14.Count) {is=struct RigVMUnitNode, template="Subtract::Execute(in A,in B,out Result)"}.Result
   ) {template="DISPATCH_RigVMDispatch_CoreEquals(in A,in B,out Result)"}
   > n16 = &RigUnit_SetTransform(Weight=1, bInitial=False, Value=@n15.Result, Item=@n14.Element, bPropagateToChildren=@MathIntEquals.Result)
   BranchNode_3_RigVMUnitNode = &RigVMFunction_ControlFlowBranch(ExecuteContext=@n16.ExecutePin, Condition=RigVMVariableNode() {variable=debug}.Value)
   /Script/ControlRig.RigUnit_DebugTransformMutableItemSpace(
    Color=(R=1,G=0,B=0,A=1)
    DebugDrawSettings=(DepthPriority=SDPG_Foreground,Lifetime=-1)
    Mode=Axes
    Scale=10
    Space=(Type=Bone)
    Thickness=0.15
    WorldOffset=(Rotation=(X=0,Y=0,Z=0,W=1),Translation=(X=0,Y=0,Z=0),Scale3D=(X=1,Y=1,Z=1))
    bEnabled=True
    Transform=@n15.Result
    ExecutePin=@BranchNode_3_RigVMUnitNode.True
   ) {is=struct RigVMUnitNode}
   Branch_1_RigVMUnitNode = &RigVMFunction_ControlFlowBranch(ExecuteContext=@RigUnit_SequenceExecution_1.B, Condition=RigVMVariableNode() {variable=Stretch}.Value)
   n17 = RigVMDispatchNode(Index=-1, Array=RigVMVariableNode() {variable=Controls}.Value) {template="DISPATCH_RigVMDispatch_ArrayGetAtIndex(in Array,in Index,out Element)"}
   n18 = RigVMDispatchNode(Index=-1, Array=RigVMVariableNode() {variable=Bones}.Value) {template="DISPATCH_RigVMDispatch_ArrayGetAtIndex(in Array,in Index,out Element)"}
   &RigUnit_ParentConstraint(
    AdvancedSettings=(RotationOrderForFilter=XZY)
    Filter=(TranslationFilter=(bX=false,bY=false,bZ=false),RotationFilter=(bX=True,bY=True,bZ=True),ScaleFilter=(bX=True,bY=True,bZ=True))
    Parents=((Item=(Type=Bone),Weight=1))
    Weight=1
    Parents.0.Item=@n17.Element
    Child=@n18.Element
    ExecutePin=@Branch_1_RigVMUnitNode.False
   )
   &RigUnit_ParentConstraint(
    AdvancedSettings=(RotationOrderForFilter=XZY)
    Parents=((Item=(Type=Bone),Weight=1))
    Weight=1
    Parents.0.Item=@n17.Element
    Child=@n18.Element
    ExecutePin=@Branch_1_RigVMUnitNode.True
   )
  ]}
 ContainedGraph {nodes [
   n1 = RigVMFunctionEntryNode()
   RigUnit_SequenceExecution_2 = &RigVMFunction_Sequence(ExecuteContext=@n1.ExecuteContext)
   RigUnit_SequenceExecution_2_1 = &RigVMFunction_Sequence(ExecuteContext=@RigUnit_SequenceExecution_2.B)
   RigUnit_SequenceExecution_2_2 = &RigVMFunction_Sequence(ExecuteContext=@RigUnit_SequenceExecution_2_1.B)
   RigVMFunctionReturnNode(A=@RigUnit_SequenceExecution_2.A, B=@RigUnit_SequenceExecution_2_1.A, C=@RigUnit_SequenceExecution_2_2.A, D=@RigUnit_SequenceExecution_2_2.B)
  ]}
 ContainedGraph {nodes [
   n1 = RigVMFunctionEntryNode()
   RigUnit_SequenceExecution_4 = &RigVMFunction_Sequence(ExecuteContext=@n1.ExecuteContext)
   RigUnit_SequenceExecution_4_1 = &RigVMFunction_Sequence(ExecuteContext=@RigUnit_SequenceExecution_4.B)
   RigUnit_SequenceExecution_4_2 = &RigVMFunction_Sequence(ExecuteContext=@RigUnit_SequenceExecution_4_1.B)
   RigVMFunctionReturnNode(A=@RigUnit_SequenceExecution_4.A, B=@RigUnit_SequenceExecution_4_1.A, C=@RigUnit_SequenceExecution_4_2.A, D=@RigUnit_SequenceExecution_4_2.B)
  ]}
 ContainedGraph {nodes [
   n1 = RigVMFunctionEntryNode()
   RigUnit_SequenceExecution = &RigVMFunction_Sequence(ExecuteContext=@n1.ExecuteContext)
   RigUnit_SequenceExecution_1 = &RigVMFunction_Sequence(ExecuteContext=@RigUnit_SequenceExecution.B)
   RigUnit_SequenceExecution_2 = &RigVMFunction_Sequence(ExecuteContext=@RigUnit_SequenceExecution_1.B)
   RigVMFunctionReturnNode(A=@RigUnit_SequenceExecution.A, B=@RigUnit_SequenceExecution_1.A, C=@RigUnit_SequenceExecution_2.A, D=@RigUnit_SequenceExecution_2.B)
  ]}
 ContainedGraph {nodes [
   n1 = RigVMFunctionEntryNode()
   RigUnit_SequenceExecution_1 = &RigVMFunction_Sequence(ExecuteContext=@n1.ExecuteContext)
   RigUnit_SequenceExecution_1_1 = &RigVMFunction_Sequence(ExecuteContext=@RigUnit_SequenceExecution_1.B)
   RigUnit_SequenceExecution_1_2 = &RigVMFunction_Sequence(ExecuteContext=@RigUnit_SequenceExecution_1_1.B)
   RigVMFunctionReturnNode(A=@RigUnit_SequenceExecution_1.A, B=@RigUnit_SequenceExecution_1_1.A, C=@RigUnit_SequenceExecution_1_2.A, D=@RigUnit_SequenceExecution_1_2.B)
  ]}
 ContainedGraph {nodes [
   n1 = RigVMFunctionEntryNode()
   RigUnit_SequenceExecution = RigVMAggregateNode(ExecuteContext=@n1.Execute)
   Branch_RigVMUnitNode = &RigVMFunction_ControlFlowBranch(Condition=@n1.Stretch, ExecuteContext=@RigUnit_SequenceExecution.A)
   n2 = &RigUnit_ControlRigSplineFromPoints(bClosed=False, Points=@n1.Points, SamplesPerSegment=@n1.Samples Per Segment, SplineMode=@n1.Spline Mode)
   &RigUnit_FitChainToSplineCurveItemArray(
    Alignment=Front
    DebugSettings=(Scale=1,CurveColor=(R=1,G=1,B=0,A=1),SegmentsColor=(R=1,G=0,B=0,A=1),WorldOffset=(Rotation=(X=0,Y=0,Z=0,W=1),Translation=(X=0,Y=0,Z=0),Scale3D=(X=1,Y=1,Z=1)))
    Maximum=1
    Rotations=()
    SamplingPrecision=12
    Weight=1
    bPropagateToChildren=True
    ExecutePin=@Branch_RigVMUnitNode.False
    Spline=@n2.Spline
    Items=@n1.Items
    PoleVectorPosition=@n1.Pole Vector Position
    PrimaryAxis=@n1.Primary Axis
    SecondaryAxis=@n1.Secondary Axis
   )
   &RigUnit_FitChainToSplineCurveItemArray(
    Alignment=Stretched
    DebugSettings=(Scale=1,CurveColor=(R=1,G=1,B=0,A=1),SegmentsColor=(R=1,G=0,B=0,A=1),WorldOffset=(Rotation=(X=0,Y=0,Z=0,W=1),Translation=(X=0,Y=0,Z=0),Scale3D=(X=1,Y=1,Z=1)))
    Maximum=1
    Rotations=()
    SamplingPrecision=12
    Weight=1
    bPropagateToChildren=True
    ExecutePin=@Branch_RigVMUnitNode.True
    Spline=@n2.Spline
    Items=@n1.Items
    PoleVectorPosition=@n1.Pole Vector Position
    PrimaryAxis=@n1.Primary Axis
    SecondaryAxis=@n1.Secondary Axis
   )
   BranchNode_RigVMUnitNode = &RigVMFunction_ControlFlowBranch(Condition=@n1.Draw Spline, ExecuteContext=@RigUnit_SequenceExecution.B)
   &RigUnit_DrawControlRigSpline(Color=(R=1,G=0,B=0,A=1), Detail=16, ExecutePin=@BranchNode_RigVMUnitNode.True, Spline=@n2.Spline, Thickness=@n1.Draw Spline Thickness)
   RigVMFunctionReturnNode(Spline=@n2.Spline, Execute=@RigUnit_SequenceExecution.D)
  ]}
 ContainedGraph {nodes [
   n1 = RigVMFunctionEntryNode()
   RigUnit_SequenceExecution = &RigVMFunction_Sequence(ExecuteContext=@n1.ExecuteContext)
   RigUnit_SequenceExecution_1 = &RigVMFunction_Sequence(ExecuteContext=@RigUnit_SequenceExecution.B)
   RigUnit_SequenceExecution_2 = &RigVMFunction_Sequence(ExecuteContext=@RigUnit_SequenceExecution_1.B)
   RigVMFunctionReturnNode(A=@RigUnit_SequenceExecution.A, B=@RigUnit_SequenceExecution_1.A, C=@RigUnit_SequenceExecution_2.A, D=@RigUnit_SequenceExecution_2.B)
  ]}
 ContainedGraph {nodes [
   n1 = RigVMFunctionEntryNode()
   n2 = RigVMDispatchNode(Num=4, Array=RigVMVariableNode() {variable=Output Points}.Value, ExecuteContext=@n1.Execute) {template="DISPATCH_RigVMDispatch_ArraySetNum(io Array,in Num)"}
   n3 = RigVMVariableNode() {variable=Point 0}
   > n4 = RigVMDispatchNode(Index=0, Array=@n2.Array, Element=@n3.Value) {template="DISPATCH_RigVMDispatch_ArraySetAtIndex(io Array,in Index,in Element)"}
   n5 = RigVMVariableNode() {variable=t}
   n6 = RigVMVariableNode() {variable=t}
   n7 = /Script/RigVM.RigVMFunction_MathFloatSub(A=1, B=@n6.Value) {is=struct RigVMUnitNode, template="Subtract::Execute(in A,in B,out Result)"}
   Get A and C = RigVMCollapseNode(order=3, Start Point=@n1.Point 0, Middle Point=@n1.Point 1, End Point=@n1.Point 2, t=@n1.t)
   n8 = RigVMDispatchNode(
    Condition=true
    False=3
    True=&RigVMFunction_MathVectorDistance(B=@"Get A and C".C, A=RigVMVariableNode() {variable=Point 1}.Value).Result
   ) {template="DISPATCH_RigVMDispatch_If(in Condition,in True,in False,out Result)"}
   n9 = /Script/RigVM.RigVMFunction_MathVectorUnit(Value=/Script/RigVM.RigVMFunction_MathVectorSub(B=RigVMVariableNode() {variable=Point 0}.Value, A=RigVMVariableNode() {variable=Point 2}.Value) {is=struct RigVMUnitNode, template="Subtract::Execute(in A,in B,out Result)"}.Result) {is=struct RigVMUnitNode, template="Unit::Execute(in Value,out Result)"}
   n10 = RigVMVariableNode() {variable=Point 1}
   n11 = &RigVMFunction_MathVectorSub(
    B=/Script/RigVM.RigVMFunction_MathVectorScale(Factor=@n8.Result, Value=@n9.Result) {is=struct RigVMUnitNode, template="Scale::Execute(in Value,in Factor,out Result)"}.Result
    A=@n10.Value
   ) {template="Subtract::Execute(in A,in B,out Result)"}
   n12 = /Script/RigVM.RigVMFunction_MathVectorAdd(
    B=&RigVMFunction_MathVectorScale(
     Factor=/Script/RigVM.RigVMFunction_MathFloatDiv(A=1, B=@n7.Result) {is=struct RigVMUnitNode, template="Divide::Execute(in A,in B,out Result)"}.Result
     Value=&RigVMFunction_MathVectorSub(A=@n11.Result, B=@"Get A and C".A) {template="Subtract::Execute(in A,in B,out Result)"}.Result
    ) {template="Scale::Execute(in Value,in Factor,out Result)"}.Result
    A=@"Get A and C".A
   ) {is=struct RigVMUnitNode, template="Add::Execute(in A,in B,out Result)"}
   n13 = RigVMVariableNode() {variable=Point 0}
   n14 = &RigVMFunction_MathVectorAdd(
    B=&RigVMFunction_MathVectorScale(
     Factor=&RigVMFunction_MathFloatDiv(A=1, B=@n5.Value) {template="Divide::Execute(in A,in B,out Result)"}.Result
     Value=&RigVMFunction_MathVectorSub(A=@n12.Result, B=@n13.Value) {template="Subtract::Execute(in A,in B,out Result)"}.Result
    ) {template="Scale::Execute(in Value,in Factor,out Result)"}.Result
    A=@n13.Value
   ) {template="Add::Execute(in A,in B,out Result)"}
   > n15 = RigVMDispatchNode(Index=1, Array=@n4.Array, Element=@n14.Result) {template="DISPATCH_RigVMDispatch_ArraySetAtIndex(io Array,in Index,in Element)"}
   n16 = &RigVMFunction_MathFloatSub(A=1, B=@n5.Value) {template="Subtract::Execute(in A,in B,out Result)"}
   n17 = RigVMVariableNode() {variable=t}
   n18 = &RigVMFunction_MathFloatMul(A=@n8.Result, B=&RigVMFunction_MathFloatSub(A=1, B=@n17.Value) {template="Subtract::Execute(in A,in B,out Result)"}.Result) {template="Multiply::Execute(in A,in B,out Result)"}
   n19 = &RigVMFunction_MathVectorAdd(
    B=&RigVMFunction_MathVectorScale(
     Factor=&RigVMFunction_MathFloatDiv(A=@n18.Result, B=@n17.Value) {template="Divide::Execute(in A,in B,out Result)"}.Result
     Value=@n9.Result
    ) {template="Scale::Execute(in Value,in Factor,out Result)"}.Result
    A=@n10.Value
   ) {template="Add::Execute(in A,in B,out Result)"}
   n20 = &RigVMFunction_MathVectorAdd(
    B=&RigVMFunction_MathVectorScale(
     Factor=&RigVMFunction_MathFloatDiv(A=1, B=@n6.Value) {template="Divide::Execute(in A,in B,out Result)"}.Result
     Value=&RigVMFunction_MathVectorSub(A=@n19.Result, B=@"Get A and C".A) {template="Subtract::Execute(in A,in B,out Result)"}.Result
    ) {template="Scale::Execute(in Value,in Factor,out Result)"}.Result
    A=@"Get A and C".A
   ) {template="Add::Execute(in A,in B,out Result)"}
   n21 = RigVMVariableNode() {variable=Point 2}
   n22 = &RigVMFunction_MathVectorAdd(
    B=&RigVMFunction_MathVectorScale(
     Factor=&RigVMFunction_MathFloatDiv(A=1, B=@n16.Result) {template="Divide::Execute(in A,in B,out Result)"}.Result
     Value=&RigVMFunction_MathVectorSub(A=@n20.Result, B=@n21.Value) {template="Subtract::Execute(in A,in B,out Result)"}.Result
    ) {template="Scale::Execute(in Value,in Factor,out Result)"}.Result
    A=@n21.Value
   ) {template="Add::Execute(in A,in B,out Result)"}
   > n23 = RigVMDispatchNode(Index=2, Array=@n15.Array, Element=@n22.Result) {template="DISPATCH_RigVMDispatch_ArraySetAtIndex(io Array,in Index,in Element)"}
   n24 = RigVMVariableNode() {variable=Point 2}
   > n25 = RigVMDispatchNode(Index=3, Array=@n23.Array, Element=@n24.Value) {template="DISPATCH_RigVMDispatch_ArraySetAtIndex(io Array,in Index,in Element)"}
   RerouteNode = RigVMDispatchNode(
    Values=((X=0,Y=0,Z=0),((X=0,Y=0,Z=0),X=0,Y=0,Z=0),((X=0,Y=0,Z=0),X=0,Y=0,Z=0),((X=0,Y=0,Z=0),X=0,Y=0,Z=0))
    Values.1=@n14.Result
    Values.2=@n22.Result
    Values.3=@n24.Value
    Values.0=@n3.Value
   ) {template="DISPATCH_RigVMDispatch_ArrayMake(in Values,out Array)"}
   RigVMFunctionReturnNode(
    Control Points=@n25.Array
    Execute=@n25.ExecuteContext
    Spline=&RigUnit_ControlRigSplineFromPoints(SamplesPerSegment=16, SplineMode=BSpline, bClosed=False, Points=@RerouteNode.Array).Spline
   )
  ]}
 ContainedGraph {nodes [
   n1 = RigVMFunctionEntryNode()
   Projection Ratio = RigVMCollapseNode(order=@n1.order, t=@n1.t)
   n2 = &RigVMFunction_MathVectorAdd(
    A=&RigVMFunction_MathVectorScale(Factor=@"Projection Ratio".Result, Value=RigVMVariableNode() {variable=Start Point}.Value) {template="Scale::Execute(in Value,in Factor,out Result)"}.Result
    B=&RigVMFunction_MathVectorScale(
     Factor=&RigVMFunction_MathFloatSub(A=1, B=@"Projection Ratio".Result) {template="Subtract::Execute(in A,in B,out Result)"}.Result
     Value=RigVMVariableNode() {variable=End Point}.Value
    ) {template="Scale::Execute(in Value,in Factor,out Result)"}.Result
   ) {template="Add::Execute(in A,in B,out Result)"}
   ABC Ratio = RigVMCollapseNode(order=@n1.order, t=@n1.t)
   n3 = &RigVMFunction_MathFloatDiv(A=1, B=@"ABC Ratio".Result) {template="Divide::Execute(in A,in B,out Result)"}
   n4 = RigVMVariableNode() {variable=Middle Point}
   n5 = &RigVMFunction_MathVectorSub(B=@n2.Result, A=@n4.Value) {template="Subtract::Execute(in A,in B,out Result)"}
   RigVMFunctionReturnNode(
    C=@n2.Result
    A=&RigVMFunction_MathVectorAdd(
     B=&RigVMFunction_MathVectorScale(Factor=@n3.Result, Value=@n5.Result) {template="Scale::Execute(in Value,in Factor,out Result)"}.Result
     A=@n4.Value
    ) {template="Add::Execute(in A,in B,out Result)"}.Result
   )
  ]}
 ContainedGraph {nodes [
   n1 = RigVMFunctionEntryNode()
   n2 = /Script/RigVM.RigVMFunction_MathIntToFloat(Value=@n1.order) {is=struct RigVMUnitNode, template="Cast::Execute(in Value,out Result)"}
   n3 = /Script/RigVM.RigVMFunction_MathFloatPow(A=&RigVMFunction_MathFloatSub(A=1, B=@n1.t) {template="Subtract::Execute(in A,in B,out Result)"}.Result, B=@n2.Result) {is=struct RigVMUnitNode, template="Power::Execute(in A,in B,out Result)"}
   n4 = &RigVMFunction_MathFloatPow(A=@n1.t, B=@n2.Result) {template="Power::Execute(in A,in B,out Result)"}
   RigVMFunctionReturnNode(Result=&RigVMFunction_MathFloatDiv(B=&RigVMFunction_MathFloatAdd(A=@n3.Result, B=@n4.Result) {template="Add::Execute(in A,in B,out Result)"}.Result, A=@n3.Result) {template="Divide::Execute(in A,in B,out Result)"}.Result)
  ]}
 ContainedGraph {nodes [
   n1 = RigVMFunctionEntryNode()
   n2 = &RigVMFunction_MathIntToFloat(Value=@n1.order) {template="Cast::Execute(in Value,out Result)"}
   n3 = &RigVMFunction_MathFloatAdd(
    B=&RigVMFunction_MathFloatPow(A=@n1.t, B=@n2.Result) {template="Power::Execute(in A,in B,out Result)"}.Result
    A=&RigVMFunction_MathFloatPow(A=&RigVMFunction_MathFloatSub(A=1, B=@n1.t) {template="Subtract::Execute(in A,in B,out Result)"}.Result, B=@n2.Result) {template="Power::Execute(in A,in B,out Result)"}.Result
   ) {template="Add::Execute(in A,in B,out Result)"}
   n4 = &RigVMFunction_MathFloatSub(B=1, A=@n3.Result) {template="Subtract::Execute(in A,in B,out Result)"}
   RigVMFunctionReturnNode(Result=/Script/RigVM.RigVMFunction_MathFloatAbs(Value=&RigVMFunction_MathFloatDiv(B=@n3.Result, A=@n4.Result) {template="Divide::Execute(in A,in B,out Result)"}.Result) {is=struct RigVMUnitNode, template="Absolute::Execute(in Value,out Result)"}.Result)
  ]}
 ContainedGraph {nodes [
   n1 = RigVMFunctionEntryNode()
   n2 = RigVMDispatchNode(Num=4, Array=RigVMVariableNode() {variable=Outputs}.Value, ExecuteContext=@n1.Execute) {template="DISPATCH_RigVMDispatch_ArraySetNum(io Array,in Num)"}
   n3 = RigVMVariableNode() {variable=Point 0}
   > n4 = RigVMDispatchNode(Index=0, Array=@n2.Array, Element=@n3.Value) {template="DISPATCH_RigVMDispatch_ArraySetAtIndex(io Array,in Index,in Element)"}
   n5 = RigVMVariableNode() {variable=Point 3}
   n6 = &RigVMFunction_MathVectorDistance(B=@n1.Point 0, A=@n1.Point 1)
   n7 = &RigVMFunction_MathFloatAdd(A=@n6.Result, B=&RigVMFunction_MathVectorDistance(A=@n1.Point 1, B=@n1.Point 2).Result) {template="Add::Execute(in A,in B,out Result)"}
   n8 = &RigVMFunction_MathFloatAdd(A=@n7.Result, B=&RigVMFunction_MathVectorDistance(A=@n1.Point 2, B=@n1.Point 3).Result) {template="Add::Execute(in A,in B,out Result)"}
   n9 = &RigVMFunction_MathFloatDiv(B=@n8.Result, A=@n6.Result) {template="Divide::Execute(in A,in B,out Result)"}
   b3t1 = RigVMCollapseNode(t=@n9.Result)
   n10 = RigVMVariableNode() {variable=Point 0}
   b0t1 = RigVMCollapseNode(t=@n9.Result)
   n11 = &RigVMFunction_MathVectorScale(Value=@n10.Value, Factor=@b0t1.Result) {template="Scale::Execute(in Value,in Factor,out Result)"}
   n12 = RigVMVariableNode() {variable=Point 1}
   n13 = &RigVMFunction_MathFloatDiv(A=@n7.Result, B=@n8.Result) {template="Divide::Execute(in A,in B,out Result)"}
   b3t2 = RigVMCollapseNode(t=@n13.Result)
   b0t2 = RigVMCollapseNode(t=@n13.Result)
   n14 = &RigVMFunction_MathVectorScale(Value=@n10.Value, Factor=@b0t2.Result) {template="Scale::Execute(in Value,in Factor,out Result)"}
   n15 = RigVMVariableNode() {variable=Point 2}
   b1t1 = RigVMCollapseNode(t=@n9.Result)
   b1t2 = RigVMCollapseNode(t=@n13.Result)
   b2t1 = RigVMCollapseNode(t=@n9.Result)
   b2t2 = RigVMCollapseNode(t=@n13.Result)
   Solve = RigVMCollapseNode(
    c=&RigVMFunction_MathVectorSub(
     B=&RigVMFunction_MathVectorScale(Value=@n5.Value, Factor=@b3t1.Result) {template="Scale::Execute(in Value,in Factor,out Result)"}.Result
     A=&RigVMFunction_MathVectorSub(B=@n11.Result, A=@n12.Value) {template="Subtract::Execute(in A,in B,out Result)"}.Result
    ) {template="Subtract::Execute(in A,in B,out Result)"}.Result
    f=&RigVMFunction_MathVectorSub(
     B=&RigVMFunction_MathVectorScale(Value=@n5.Value, Factor=@b3t2.Result) {template="Scale::Execute(in Value,in Factor,out Result)"}.Result
     A=&RigVMFunction_MathVectorSub(B=@n14.Result, A=@n15.Value) {template="Subtract::Execute(in A,in B,out Result)"}.Result
    ) {template="Subtract::Execute(in A,in B,out Result)"}.Result
    a=@b1t1.Result
    d=@b1t2.Result
    b=@b2t1.Result
    e=@b2t2.Result
   )
   > n16 = RigVMDispatchNode(Index=1, Array=@n4.Array, Element=@Solve.i) {template="DISPATCH_RigVMDispatch_ArraySetAtIndex(io Array,in Index,in Element)"}
   > n17 = RigVMDispatchNode(Index=2, Array=@n16.Array, Element=@Solve.j) {template="DISPATCH_RigVMDispatch_ArraySetAtIndex(io Array,in Index,in Element)"}
   n18 = RigVMVariableNode() {variable=Point 3}
   > n19 = RigVMDispatchNode(Index=3, Array=@n17.Array, Element=@n18.Value) {template="DISPATCH_RigVMDispatch_ArraySetAtIndex(io Array,in Index,in Element)"}
   RigVMFunctionReturnNode(
    Control Points=@n19.Array
    Execute=@n19.ExecuteContext
    Spline=&RigUnit_ControlRigSplineFromPoints(
     Points=((X=0,Y=0,Z=0),(X=0,Y=0,Z=0),(X=0,Y=0,Z=0),(X=0,Y=0,Z=0))
     SamplesPerSegment=16
     SplineMode=BSpline
     bClosed=False
     Points.1=@Solve.i
     Points.2=@Solve.j
     Points.0=@n3.Value
     Points.3=@n18.Value
    ).Spline
   )
  ]}
 ContainedGraph {nodes [
   RigVMFunctionEntryNode()
   n1 = &RigVMFunction_MathFloatDiv(
    A=&RigVMFunction_MathFloatMul(A=RigVMVariableNode() {variable=a}.Value, B=RigVMVariableNode() {variable=e}.Value) {template="Multiply::Execute(in A,in B,out Result)"}.Result
    B=RigVMVariableNode() {variable=d}.Value
   ) {template="Divide::Execute(in A,in B,out Result)"}
   n2 = RigVMVariableNode() {variable=b}
   n3 = &RigVMFunction_MathFloatDiv(A=RigVMVariableNode() {variable=a}.Value, B=RigVMVariableNode() {variable=d}.Value) {template="Divide::Execute(in A,in B,out Result)"}
   n4 = RigVMVariableNode() {variable=f}
   n5 = &RigVMFunction_MathVectorScale(
    Factor=&RigVMFunction_MathFloatDiv(A=1, B=&RigVMFunction_MathFloatSub(B=@n1.Result, A=@n2.Value) {template="Subtract::Execute(in A,in B,out Result)"}.Result) {template="Divide::Execute(in A,in B,out Result)"}.Result
    Value=&RigVMFunction_MathVectorSub(
     B=&RigVMFunction_MathVectorScale(Factor=@n3.Result, Value=@n4.Value) {template="Scale::Execute(in Value,in Factor,out Result)"}.Result
     A=RigVMVariableNode() {variable=c}.Value
    ) {template="Subtract::Execute(in A,in B,out Result)"}.Result
   ) {template="Scale::Execute(in Value,in Factor,out Result)"}
   n6 = RigVMVariableNode() {variable=a}
   n7 = &RigVMFunction_MathVectorScale(Value=@n5.Result, Factor=RigVMVariableNode() {variable=b}.Value) {template="Scale::Execute(in Value,in Factor,out Result)"}
   n8 = RigVMVariableNode() {variable=c}
   RigVMFunctionReturnNode(
    j=@n5.Result
    i=&RigVMFunction_MathVectorScale(
     Factor=&RigVMFunction_MathFloatDiv(A=1, B=@n6.Value) {template="Divide::Execute(in A,in B,out Result)"}.Result
     Value=&RigVMFunction_MathVectorSub(B=@n7.Result, A=@n8.Value) {template="Subtract::Execute(in A,in B,out Result)"}.Result
    ) {template="Scale::Execute(in Value,in Factor,out Result)"}.Result
   )
  ]}
 ContainedGraph {nodes [
   n1 = RigVMFunctionEntryNode()
   n2 = &RigVMFunction_MathFloatSub(A=1, B=@n1.t) {template="Subtract::Execute(in A,in B,out Result)"}
   n3 = &RigVMFunction_MathFloatMul(A=@n2.Result, B=@n2.Result) {template="Multiply::Execute(in A,in B,out Result)"}
   RigVMFunctionReturnNode(Result=&RigVMFunction_MathFloatMul(B=3, A=&RigVMFunction_MathFloatMul(A=@n1.t, B=@n3.Result) {template="Multiply::Execute(in A,in B,out Result)"}.Result) {template="Multiply::Execute(in A,in B,out Result)"}.Result)
  ]}
 ContainedGraph {nodes [
   n1 = RigVMFunctionEntryNode()
   n2 = &RigVMFunction_MathFloatMul(A=@n1.t, B=@n1.t) {template="Multiply::Execute(in A,in B,out Result)"}
   n3 = &RigVMFunction_MathFloatSub(A=1, B=@n1.t) {template="Subtract::Execute(in A,in B,out Result)"}
   RigVMFunctionReturnNode(Result=&RigVMFunction_MathFloatMul(B=3, A=&RigVMFunction_MathFloatMul(B=@n2.Result, A=@n3.Result) {template="Multiply::Execute(in A,in B,out Result)"}.Result) {template="Multiply::Execute(in A,in B,out Result)"}.Result)
  ]}
 ContainedGraph {nodes [
   n1 = RigVMFunctionEntryNode()
   n2 = &RigVMFunction_MathFloatSub(A=1, B=@n1.t) {template="Subtract::Execute(in A,in B,out Result)"}
   n3 = &RigVMFunction_MathFloatMul(A=@n2.Result, B=@n2.Result) {template="Multiply::Execute(in A,in B,out Result)"}
   RigVMFunctionReturnNode(Result=&RigVMFunction_MathFloatMul(B=3, A=&RigVMFunction_MathFloatMul(A=@n1.t, B=@n3.Result) {template="Multiply::Execute(in A,in B,out Result)"}.Result) {template="Multiply::Execute(in A,in B,out Result)"}.Result)
  ]}
 ContainedGraph {nodes [
   n1 = RigVMFunctionEntryNode()
   n2 = &RigVMFunction_MathFloatMul(A=@n1.t, B=@n1.t) {template="Multiply::Execute(in A,in B,out Result)"}
   n3 = &RigVMFunction_MathFloatSub(A=1, B=@n1.t) {template="Subtract::Execute(in A,in B,out Result)"}
   RigVMFunctionReturnNode(Result=&RigVMFunction_MathFloatMul(B=3, A=&RigVMFunction_MathFloatMul(B=@n2.Result, A=@n3.Result) {template="Multiply::Execute(in A,in B,out Result)"}.Result) {template="Multiply::Execute(in A,in B,out Result)"}.Result)
  ]}
 ContainedGraph {nodes [
   n1 = RigVMFunctionEntryNode()
   RigVMFunctionReturnNode(Result=&RigVMFunction_MathFloatPow(B=3, A=&RigVMFunction_MathFloatSub(A=1, B=@n1.t) {template="Subtract::Execute(in A,in B,out Result)"}.Result) {template="Power::Execute(in A,in B,out Result)"}.Result)
  ]}
 ContainedGraph {nodes [RigVMFunctionReturnNode(Result=&RigVMFunction_MathFloatPow(B=3, A=RigVMFunctionEntryNode().t) {template="Power::Execute(in A,in B,out Result)"}.Result)]}
 ContainedGraph {nodes [
   n1 = RigVMFunctionEntryNode()
   RigVMFunctionReturnNode(Result=&RigVMFunction_MathFloatPow(B=3, A=&RigVMFunction_MathFloatSub(A=1, B=@n1.t) {template="Subtract::Execute(in A,in B,out Result)"}.Result) {template="Power::Execute(in A,in B,out Result)"}.Result)
  ]}
 ContainedGraph {nodes [RigVMFunctionReturnNode(Result=&RigVMFunction_MathFloatPow(B=3, A=RigVMFunctionEntryNode().t) {template="Power::Execute(in A,in B,out Result)"}.Result)]}
 RigVMFunctionLibrary {nodes [
   AttachChainToSpline = RigVMCollapseNode(
    Draw Spline Thickness=1
    Execute=()
    Items=()
    Points=((X=0,Y=0,Z=0),(X=0,Y=0,Z=0),(X=0,Y=0,Z=0),(X=0,Y=0,Z=0))
    Pole Vector Position=(X=0,Y=0,Z=0)
    Primary Axis=(X=1,Y=0,Z=0)
    Samples Per Segment=16
    Secondary Axis=(X=0,Y=0,Z=0)
    Spline Mode=BSpline
   )
   SplineFromItems = RigVMCollapseNode(ExecuteContext=(), Items=(), Samples Per Segment=16, Spline Mode=Hermite)
   SplineIK = RigVMCollapseNode(
    Bones=()
    Controls=()
    ExecuteContext=()
    Primary Axis=(X=1,Y=0,Z=0)
    Samples Per Segment=16
    Secondary Spline Direction=(X=0,Y=5,Z=0)
    Spline Mode=BSpline
    Up Axis=(X=0,Y=1,Z=0)
   )
   SplineThrough3Points = RigVMCollapseNode(Execute=(), Point 0=(X=0,Y=0,Z=0), Point 1=(X=0,Y=0,Z=0), Point 2=(X=0,Y=0,Z=0), t=0.5)
   SplineThrough4Points = RigVMCollapseNode(Execute=(), Point 0=(X=0,Y=0,Z=0), Point 1=(X=0,Y=0,Z=0), Point 2=(X=0,Y=0,Z=0), Point 3=(X=0,Y=0,Z=0))
  ]}
 RigVMModel {nodes [RigUnit_BeginExecution = &RigUnit_BeginExecution()]}
]
